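; EQS query context for the FossilMeal Bread NPC boss that supplies a single target actor to environment queries.
; It casts the querier to the boss pawn and returns that pawn's LeapTarget when valid.
; Otherwise it falls back to the blackboard target actor through the shared AI function library.
; A failed cast returns no actor.
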(fn ProvideSingleActor
  (params (QuerierObject :type Object%20Reference) (QuerierActor :type Actor%20Object%20Reference))
  (bind _asbp_player_pawn_npcboss_fossil_meal_bread (Utilities|Casting|CastToBP_PlayerPawn_NPCBoss_FossilMeal_Bread QuerierActor)
    (:then
      (Utilities|IsValid (Variables|/FossilMeal_Bread/AI/Pawns/BP_PlayerPawn_NPCBoss_FossilMeal_Bread.BP_PlayerPawn_NPCBoss_FossilMeal_Bread_C|GetLeapTarget _asbp_player_pawn_npcboss_fossil_meal_bread)
        (:Is%20Valid
          (return (Variables|/FossilMeal_Bread/AI/Pawns/BP_PlayerPawn_NPCBoss_FossilMeal_Bread.BP_PlayerPawn_NPCBoss_FossilMeal_Bread_C|GetLeapTarget _asbp_player_pawn_npcboss_fossil_meal_bread)))
        (:Is%20Not%20Valid
          (bind _targetactor (Class|BPFLAICommonFunctions|GetTargetActorfromBlackboard _asbp_player_pawn_npcboss_fossil_meal_bread))
          (return _targetactor))))
    (:CastFailed
      (return 0))))
