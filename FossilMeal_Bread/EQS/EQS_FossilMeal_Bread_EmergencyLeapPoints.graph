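# EQS query that picks candidate emergency leap destinations for the FossilMeal Bread AI.
# A single donut ring of points surrounds the emergency leap location context and projects onto navmesh with a water-avoiding filter, then lifts vertically.
# Pathfinding tests each point using the Phoebe nav filter, and an obstruction test scores it relative to the querier.
# An absolute dot-product test with negative scoring weight penalizes strongly aligned directions.
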EnvQuery /FossilMeal_Bread/EQS/EQS_FossilMeal_Bread_EmergencyLeapPoints
options [{
  generator {
   name=EnvQueryGenerator_Donut_0
   class=EnvQueryGenerator_Donut
   properties {
    bUseSpiralPattern=true
    center=&/BP_QueryContext_FossilMeal_Bread_EmergencyLeapLocation.BP_QueryContext_FossilMeal_Bread_EmergencyLeapLocation_C
    innerRadius=750
    numberOfRings=1
    outerRadius=750
    pointsPerRing=8
    projectionData {
     extentX=0
     navigationFilter=&NavFilter_Phoebe_AvoidWater_C
     postProjectionVerticalOffset=250
     projectDown=500
     projectUp=2500
    }
   }
  }
  tests [
   EnvQueryTest_Pathfinding_0 {class=EnvQueryTest_Pathfinding, properties {filterClass=/Game/Athena/AI/Phoebe/NavFilter/NavFilter_Phoebe.NavFilter_Phoebe_C}}
   FortQueryTest_IsObstructed_1 {class=FortQueryTest_IsObstructed, properties {context=&EnvQueryContext_Querier, testPurpose=Score}}
   EnvQueryTest_Dot_0 {
    class=EnvQueryTest_Dot
    properties {
     bAbsoluteValue=true
     clampMaxType=None
     clampMinType=SpecifiedValue
     floatValueMax=1
     floatValueMin=0.6
     scoreClampMax=0
     scoreClampMin=0
     scoringFactor=-1
     testComment=""
    }
   }
  ]
 }]
