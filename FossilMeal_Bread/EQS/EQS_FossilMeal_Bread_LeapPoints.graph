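# EQS query that picks leap points for the FossilMeal_Bread AI around its target actor.
# A spiral donut around the target generates nav-projected points using a water-avoiding filter, raised after projection.
# Tests filter by AI leash, minimum querier distance, obstruction, pathfinding and querier-facing dot product.
# Scoring favors horizontally or vertically suitable points: farther from the querier, with little height difference, also weighing target obstruction and target facing.
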EnvQuery /FossilMeal_Bread/EQS/EQS_FossilMeal_Bread_LeapPoints
options [{
  generator {
   name=EnvQueryGenerator_Donut_1
   class=EnvQueryGenerator_Donut
   properties {
    bUseSpiralPattern=true
    center=&/BP_QueryContext_FossilMeal_Bread_TargetActor.BP_QueryContext_FossilMeal_Bread_TargetActor_C
    innerRadius=1000
    numberOfRings=3
    outerRadius=2500
    pointsPerRing=8
    projectionData {
     extentX=0
     navigationFilter=&NavFilter_Phoebe_AvoidWater_C
     postProjectionVerticalOffset=250
     projectDown=500
     projectUp=2500
    }
   }
  }
  tests [
   FortQueryTest_InsideAIBotLeash_2 {class=FortQueryTest_InsideAIBotLeash, properties {testPurpose=Filter}}
   EnvQueryTest_Distance_0 {
    class=EnvQueryTest_Distance
    properties {
     distanceTo=&EnvQueryContext_Querier
     filterType=Minimum
     floatValueMax=2000
     floatValueMin=1000
     scoringEquation=Linear
     scoringFactor=1
     testComment=""
     testMode=Distance3D
     testPurpose=Filter
    }
   }
   FortQueryTest_IsObstructed_4 {class=FortQueryTest_IsObstructed, properties {context=&EnvQueryContext_Querier, testPurpose=Filter}}
   EnvQueryTest_Pathfinding_2 {class=EnvQueryTest_Pathfinding, properties {filterClass=/Game/Athena/AI/Phoebe/NavFilter/NavFilter_Phoebe.NavFilter_Phoebe_C}}
   EnvQueryTest_Dot_0 {
    class=EnvQueryTest_Dot
    properties {
     clampMaxType=None
     clampMinType=None
     filterType=Minimum
     scoreClampMax=0
     scoreClampMin=0
     scoringFactor=1
     testComment=""
     testMode=Dot3D
     testPurpose=Filter
    }
   }
   EnvQueryTest_Distance_6 {
    class=EnvQueryTest_Distance
    properties {
     distanceTo=&EnvQueryContext_Querier
     filterType=Minimum
     floatValueMax=2000
     floatValueMin=1000
     scoringEquation=Linear
     scoringFactor=1
     testComment=""
     testMode=Distance3D
     testPurpose=Score
    }
   }
   EnvQueryTest_Distance_7 {
    class=EnvQueryTest_Distance
    properties {
     distanceTo=&EnvQueryContext_Querier
     filterType=Range
     floatValueMin=0
     scoringEquation=InverseLinear
     scoringFactor=1
     testComment=""
     testMode=DistanceZ
     testPurpose=Score
    }
   }
   FortQueryTest_IsObstructed_5 {
    class=FortQueryTest_IsObstructed
    properties {context=&BP_QueryContext_FossilMeal_Bread_TargetActor_C, testPurpose=Score}
   }
   EnvQueryTest_Dot_1 {
    class=EnvQueryTest_Dot
    properties {
     clampMaxType=None
     clampMinType=None
     filterType=Minimum
     scoreClampMax=0
     scoreClampMin=0
     scoringFactor=1
     testComment=""
     testMode=Dot3D
     lineA {rotation=&BP_QueryContext_FossilMeal_Bread_TargetActor_C}
     lineB {lineFrom=&BP_QueryContext_FossilMeal_Bread_TargetActor_C}
    }
   }
  ]
 }]
